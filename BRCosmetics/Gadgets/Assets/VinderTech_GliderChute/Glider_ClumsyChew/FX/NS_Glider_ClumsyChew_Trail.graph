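# Niagara trail effect for the ClumsyChew glider, combining front and back socket ribbons, speed-line meshes and two skeletal-mesh-attached wind meshes.
# Each ribbon emitter bursts particles partitioned across sockets, re-placing them every update along a normalized-index offset that stretches with timelines and bends with smoothed system values.
# Speed lines spawn on a torus shifted by RightDot.
# Timelines drive fades, front-end flags swap tuning, and time-of-day scalars adjust ribbons and wind.
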NiagaraSystem /BRCosmetics/Gadgets/Assets/VinderTech_GliderChute/Glider_ClumsyChew/FX/NS_Glider_ClumsyChew_Trail
dependencies {
 usedDataInterfaces [
  &NiagaraDataInterfaceArrayInt32
  &NiagaraDataInterfaceCurve
  &NiagaraDataInterfaceArrayFloat
  &NiagaraDataInterfaceSkeletalMesh
 ]
 usedDynamicInputs [
  &MakeIntFromFloat
  &NormalizeVector
  &TransformVector
 ]
 usedModules [
  &self:SystemSpawnScript.SystemScriptSource.SystemScriptGraph.NiagaraNodeAssignment_0.SetVariables_#1
  &SystemState
  &Timeline
  &FN_Integrator
  &InterpolateOverTime
 ]
}
emitters {
 NE_Ribbons_Back {
  emitterUpdateScript [
   &EmitterState(
    Life Cycle Mode=::NewEnumerator1
    Loop Behavior=::NewEnumerator1
    Loop Delay=&MakeCustomFloatFromBool(Bool=$User.bIsFrontEndPreview, True Float=1.5, False Float=1.35)
    Loop Duration Mode=::NewEnumerator1
    UseLoopDelay=true
   )
   &SpawnBurst_Instantaneous(Spawn Count=50, Use Spawn Probability=false, UseLoopCountLimit=true)
  ]
  particleSpawnScript [
   &PartitionParticles(Write Partition to Ribbon ID=::NewEnumerator0)
   /SetVariables(
    Particles.RibbonID=/Niagara/DynamicInputs/Meshes/MakeRibbonIDFromSpawnGroup()
    Particles.RibbonLinkOrder:NiagaraFloat=&ReturnNormalizedExecIndex()
    Particles.SpawnGroup=$Particles.PartitionParticles.Partition
   ) {isSetParameters=true, module=SetVariables_#2}
   &InitializeParticle(
    Color Mode=::NewEnumerator0
    Lifetime=0.2
    Position Offset=<0 0 25>
    Ribbon Facing Mode=::NewEnumerator1
    Ribbon Facing Vector:Vector3f=<0 1 1>
    Ribbon Width=&MergeHelper_FloatSwitch(
     0=&MakeCustomFloatFromBool(Bool=$User.bIsFrontEndPreview, True Float=-50, False Float=-65)
     1=&MakeCustomFloatFromBool(Bool=$User.bIsFrontEndPreview, True Float=50, False Float=65)
     2=&MakeCustomFloatFromBool(Bool=$User.bIsFrontEndPreview, True Float=25, False Float=50)
     3=&MakeCustomFloatFromBool(Bool=$User.bIsFrontEndPreview, True Float=25, False Float=50)
     Index=$Particles.PartitionParticles.Partition
     Num=::NewEnumerator0
    )
    Ribbon Width Mode=::NewEnumerator1
    Sprite Size Mode=::NewEnumerator3
    Uniform Sprite Size=5
    UsePositionOffset=false
    Write Material Random=true
   )
  ]
  particleUpdateScript [
   &ParticleState(Kill Particles When Lifetime Has Elapsed=false, Let Infinitely Lived Particles Die When Emitter Deactivates=true)
   &SkeletalMeshLocation(
    Bone / Socket Index=&SelectIntFromArray(Direct Array Index=$Particles.PartitionParticles.Partition, Int Selection Array=DataInterface() {propertyValues {IntData [3,4]}})
    MaskSpawnGroup=false
    Mesh Sampling Type=::NewEnumerator5
    Sampled Position Offset=<0 -5 0>
    Sampled Position Offset Coordinate Space:ENiagaraSkelMeshTransforms=::NewEnumerator1
    Skeletal Mesh=$System.Skeletal Mesh
    Socket Sampling Mode=::NewEnumerator5
   ) {module=SkeletalMeshLocation001}
   /SetVariables(Particles.Position=&LerpPosition(
     A=$Particles.Position
     B=&AddVectorToPosition(
      Position=$Particles.Position
      Vector=&MakeCustomVectorFromBool(
       Bool=$User.bIsFrontEndPreview
       True Vector=<0 -200 -100>
       False Vector=&MakeVector(
        X=&FloatFromCurve(
         FloatCurve=DataInterface() {propertyValues {Curve {keys [<at 0 0 RCIM_Cubic>, {arriveTangent=2.1213, interpMode=RCIM_Cubic, leaveTangent=2.1213, tangentMode=RCTM_User, time=1, value=1}]}}}
         CurveIndex=&ReturnNormalizedExecIndex()
         Scale Curve=&Multiply_Float(
          A=&MakeFloatFromVector(Vector=$System.InterpolateOverTime.InterpolateOverTime_V3_.Moving Average, Channel=::NewEnumerator0)
          B=&Lerp_Float(A=0, B=50, Alpha=$System.InterpolateOverTime001.InterpolateOverTime_Float.Moving Average)
         )
        )
        Y=&FloatFromCurve(
         FloatCurve=DataInterface() {propertyValues {Curve {keys [{tangentMode=RCTM_User, time=0, value=0}, {tangentMode=RCTM_User, time=1, value=1}]}}}
         CurveIndex=&ReturnNormalizedExecIndex()
         Scale Curve=&FN_FrontEndFloatModifier(
          Gameplay=&Lerp_Float(
           A=&SelectFloatFromArray(Direct Array Index=$Particles.PartitionParticles.Partition, Float Selection Array=DataInterface() {propertyValues {FloatData [50,50,50,50]}})
           B=&SelectFloatFromArray(
            Direct Array Index=$Particles.PartitionParticles.Partition
            Float Selection Array=DataInterface() {propertyValues {FloatData [200,200,150,150]}}
           )
           Alpha=$System.Timeline.NormalizedTime
          )
          Front End=&Lerp_Float(A=0, B=200, Alpha=$System.Timeline002.NormalizedTime)
         )
        )
        Z=0
       )
      )
     )
     Alpha=&ReturnNormalizedExecIndex()
    )) {isSetParameters=true, module=SetVariables_#3}
   &ScaleColor(
    Scale Alpha=&FloatFromCurve(
     FloatCurve=DataInterface() {propertyValues {Curve {keys [{tangentMode=RCTM_User, time=0, value=1}, {tangentMode=RCTM_User, time=1, value=-0.1}]}}}
     CurveIndex=$Particles.Initial.RibbonLinkOrder
     Scale Curve=&FN_FrontEndFloatModifier(
      Gameplay=&Lerp_Float(A=0, B=1, Alpha=$System.Timeline.NormalizedTime)
      Front End=&Lerp_Float(A=0, B=1, Alpha=$System.Timeline002.NormalizedTime)
     )
    )
    ScaleA=true
    ScaleRGB=true
   ) {module=ScaleColor001}
   &ScaleRibbonWidth(Ribbon Width Scale=&MakeCustomFloatFromBool(
     Bool=&SetBoolByFloatComparison(
      A=&MakeCustomFloatFromBool(Bool=$User.bIsFrontEndPreview, True Float=0, False Float=$System.Timeline001.NormalizedTime)
      B=&Multiply_Float(A=$System.Timeline.Time, B=8)
     )
     True Float=0
     False Float=&FloatFromCurve(
      FloatCurve=DataInterface() {propertyValues {Curve {defaultValue=0.5, keys [{tangentMode=RCTM_User, time=0, value=1}, {tangentMode=RCTM_User, time=1, value=0.5}]}}}
      CurveIndex=$Particles.Initial.RibbonLinkOrder
     )
    ))
   &DynamicMaterialParameters(
    Index 1 Param 1=&FNTimeOfDayScalar(Night Scale=0.15, Morning Scale=0.3, Day Scale=0.3, Evening Scale=0.3)
    Module.Index1Float0Write=true
    Module.Index1Float1Write=true
    Module.Index1Float2Write=true
    Module.Index1Float3Write=true
    Write Parameter Index 0=::NewEnumerator0
    Write Parameter Index 1=::NewEnumerator1
   )
  ]
  rendererClasses [&NiagaraRibbonRendererProperties]
  renderers [{rendererClass=&NiagaraRibbonRendererProperties}]
 }
 NE_Ribbons_Front {
  emitterUpdateScript [
   &EmitterState(
    Life Cycle Mode=::NewEnumerator1
    Loop Behavior=::NewEnumerator1
    Loop Delay=&MakeCustomFloatFromBool(Bool=$User.bIsFrontEndPreview, True Float=1.5, False Float=1.35)
    Loop Duration Mode=::NewEnumerator1
    UseLoopDelay=true
   )
   &SpawnBurst_Instantaneous(Spawn Count=50, Use Spawn Probability=false, UseLoopCountLimit=true)
  ]
  particleSpawnScript [
   &PartitionParticles(Write Partition to Ribbon ID=::NewEnumerator0)
   /SetVariables(
    Particles.RibbonID=&MakeRibbonIDFromSpawnGroup()
    Particles.RibbonLinkOrder:NiagaraFloat=&ReturnNormalizedExecIndex()
    Particles.SpawnGroup=$Particles.PartitionParticles.Partition
   ) {isSetParameters=true, module=SetVariables_#2}
   &InitializeParticle(
    Color Mode=::NewEnumerator0
    Lifetime=0.2
    Position Offset=<0 0 25>
    Ribbon Facing Mode=::NewEnumerator1
    Ribbon Facing Vector:Vector3f=<0 1 1>
    Ribbon Width=&MergeHelper_FloatSwitch(
     0=&MakeCustomFloatFromBool(Bool=$User.bIsFrontEndPreview, True Float=-50, False Float=-90)
     1=&MakeCustomFloatFromBool(Bool=$User.bIsFrontEndPreview, True Float=50, False Float=90)
     2=&MakeCustomFloatFromBool(Bool=$User.bIsFrontEndPreview, True Float=25, False Float=50)
     3=&MakeCustomFloatFromBool(Bool=$User.bIsFrontEndPreview, True Float=25, False Float=50)
     Index=$Particles.PartitionParticles.Partition
     Num=::NewEnumerator0
    )
    Ribbon Width Mode=::NewEnumerator1
    Sprite Size Mode=::NewEnumerator3
    Uniform Sprite Size=5
    UsePositionOffset=false
    Write Material Random=true
   )
  ]
  particleUpdateScript [
   &ParticleState(Kill Particles When Lifetime Has Elapsed=false, Let Infinitely Lived Particles Die When Emitter Deactivates=true)
   &SkeletalMeshLocation(
    Bone / Socket Index=&SelectIntFromArray(Direct Array Index=$Particles.PartitionParticles.Partition, Int Selection Array=DataInterface() {propertyValues {IntData [1,2]}})
    MaskSpawnGroup=false
    Mesh Sampling Type=::NewEnumerator5
    Sampled Position Offset Coordinate Space:ENiagaraSkelMeshTransforms=::NewEnumerator1
    Skeletal Mesh=$System.Skeletal Mesh
    Socket Sampling Mode=::NewEnumerator5
   ) {module=SkeletalMeshLocation001}
   /SetVariables(Particles.Position=&LerpPosition(
     A=$Particles.Position
     B=&AddVectorToPosition(
      Position=$Particles.Position
      Vector=&MakeCustomVectorFromBool(
       Bool=$User.bIsFrontEndPreview
       True Vector=<0 -200 -100>
       False Vector=&MakeVector(
        X=&FloatFromCurve(
         FloatCurve=DataInterface() {propertyValues {Curve {keys [<at 0 0 RCIM_Cubic>, {arriveTangent=2.1213, interpMode=RCIM_Cubic, leaveTangent=2.1213, tangentMode=RCTM_User, time=1, value=1}]}}}
         CurveIndex=&ReturnNormalizedExecIndex()
         Scale Curve=&Multiply_Float(
          A=&MakeFloatFromVector(Vector=$System.InterpolateOverTime.InterpolateOverTime_V3_.Moving Average, Channel=::NewEnumerator0)
          B=&Lerp_Float(A=0, B=50, Alpha=$System.InterpolateOverTime001.InterpolateOverTime_Float.Moving Average)
         )
        )
        Y=&FloatFromCurve(
         FloatCurve=DataInterface() {propertyValues {Curve {keys [{tangentMode=RCTM_User, time=0, value=0}, {tangentMode=RCTM_User, time=1, value=1}]}}}
         CurveIndex=&ReturnNormalizedExecIndex()
         Scale Curve=&FN_FrontEndFloatModifier(
          Gameplay=&Lerp_Float(
           A=&SelectFloatFromArray(Direct Array Index=$Particles.PartitionParticles.Partition, Float Selection Array=DataInterface() {propertyValues {FloatData [50,50,50,50]}})
           B=&SelectFloatFromArray(
            Direct Array Index=$Particles.PartitionParticles.Partition
            Float Selection Array=DataInterface() {propertyValues {FloatData [150,150,150,150]}}
           )
           Alpha=$System.Timeline.NormalizedTime
          )
          Front End=&Lerp_Float(A=0, B=200, Alpha=$System.Timeline002.NormalizedTime)
         )
        )
        Z=0
       )
      )
     )
     Alpha=&ReturnNormalizedExecIndex()
    )) {isSetParameters=true, module=SetVariables_#3}
   &ScaleColor(
    Scale Alpha=&FloatFromCurve(
     FloatCurve=DataInterface() {propertyValues {Curve {keys [{tangentMode=RCTM_User, time=0, value=1}, {tangentMode=RCTM_User, time=1, value=-0.1}]}}}
     CurveIndex=$Particles.Initial.RibbonLinkOrder
     Scale Curve=&FN_FrontEndFloatModifier(
      Gameplay=&Lerp_Float(A=0, B=1, Alpha=$System.Timeline.NormalizedTime)
      Front End=&Lerp_Float(A=0, B=1, Alpha=$System.Timeline002.NormalizedTime)
     )
    )
    ScaleA=true
    ScaleRGB=true
   ) {module=ScaleColor001}
   &ScaleRibbonWidth(Ribbon Width Scale=&MakeCustomFloatFromBool(
     Bool=&SetBoolByFloatComparison(
      A=&MakeCustomFloatFromBool(Bool=$User.bIsFrontEndPreview, True Float=0, False Float=$System.Timeline001.NormalizedTime)
      B=&Multiply_Float(A=$System.Timeline.Time, B=8)
     )
     True Float=0
     False Float=&FloatFromCurve(
      FloatCurve=DataInterface() {propertyValues {Curve {defaultValue=0.5, keys [{tangentMode=RCTM_User, time=0, value=1}, {tangentMode=RCTM_User, time=1, value=0.5}]}}}
      CurveIndex=$Particles.Initial.RibbonLinkOrder
     )
    ))
   &DynamicMaterialParameters(
    Index 1 Param 1=&FNTimeOfDayScalar(Night Scale=0.3, Morning Scale=0.35, Day Scale=0.5, Evening Scale=0.35)
    Module.Index1Float0Write=true
    Module.Index1Float1Write=true
    Module.Index1Float2Write=true
    Module.Index1Float3Write=true
    Write Parameter Index 0=::NewEnumerator0
    Write Parameter Index 1=::NewEnumerator1
   )
  ]
  rendererClasses [&NiagaraRibbonRendererProperties]
  renderers [{rendererClass=&NiagaraRibbonRendererProperties}]
 }
 NE_SpeedLines {
  emitterUpdateScript [
   &EmitterState(Life Cycle Mode=::NewEnumerator1, Loop Behavior=::NewEnumerator1, Loop Delay=2.5, Loop Duration Mode=::NewEnumerator1, UseLoopDelay=true)
   &SpawnRate(
    SpawnGroup=1
    SpawnRate=&Lerp_Float(
     A=&FN_FrontEndFloatModifier(
      Gameplay=0
      Front End=&FloatFromCurve(FloatCurve=DataInterface() {propertyValues {Curve {keys [<at 0.5 0>,<at 2 1>]}}}, CurveIndex=$Engine.System.Age, Scale Curve=75)
     )
     B=200
     Alpha=$System.Timeline.NormalizedTime
    )
    Use Spawn Probability=true
   ) {module=SpawnRate001}
  ]
  particleSpawnScript [
   &InitializeParticle(
    Color=<rgba 1 1 1 5>
    Mesh Scale Max=<50 1 1>
    Mesh Scale Min=<15 1 1>
    Mesh Scale Mode=::NewEnumerator2
    UsePositionOffset=false
    Write Material Random=true
   )
   &InitialMeshOrientation(
    Mesh Orientation Mode=::NewEnumerator2
    Rotation=&RandomRangeVector(Minimum=<0 0 0>, Maximum=<1 0 0>)
    Use Rotation Vector=true
   )
   &TorusLocation(
    Align Torus With Axis=true
    Handle Radius=0
    Large Radius:NiagaraFloat=&FN_FrontEndFloatModifier(
     Gameplay=&RandomRangeFloat(Minimum=220, Maximum=240)
     Front End=&RandomRangeFloat(Minimum=180, Maximum=220)
    )
    MaskSpawnGroup=false
    Module.Override Randomness=false
    Offset=<0 -180 0>
    Spawn Group Mask=1
    Torus Axis=<0 1 0>
    Torus Origin:NiagaraPosition=&Add_Vector(
     A=&FN_FrontEndVector3Modifier(Gameplay=<0 -100 50>, Front End=<0 500 100>)
     B=&Lerp_Vector(A=<0 0 0>, B=<200 0 0>, Alpha=$User.RightDot)
    )
   ) {module=TorusLocation001}
   &AddVelocityFromPoint(
    Module.Use Falloff Distance=false
    Origin Offset=&Add_Vector(A=<0 -500 0>, B=&Lerp_Vector(A=<0 0 0>, B=<300 0 0>, Alpha=$User.RightDot))
    Velocity Strength=&FN_FrontEndFloatModifier(Gameplay=150, Front End=-150)
   )
   &AddVelocity(
    Module.Use Falloff Distance=false
    Origin Offset=&Add_Vector(A=<0 -500 0>, B=&Lerp_Vector(A=<0 0 0>, B=<500 0 0>, Alpha=$User.RightDot))
    Override Seed=false
    Velocity Mode=::NewEnumerator1
    Velocity Speed=200
   ) {module=AddVelocity001}
   &AddVelocity(Velocity=&Multiply_Vector(
     A=&MakeCustomVectorFromBool(
      Bool=$User.bIsFrontEndPreview
      True Vector=&RandomRangeVector(Minimum=<0 -500 -1000>, Maximum=<0 -500 -1000>)
      False Vector=&RandomRangeVector(Minimum=<1 400 1>, Maximum=<1 600 1>)
     )
     B=&MakeVector(X=1, Y=&MakeCustomFloatFromBool(Bool=$User.bIsFrontEndPreview, True Float=5, False Float=1), Z=1)
    ))
  ]
  particleUpdateScript [
   &ScaleVelocity(Velocity Scale=&VectorFromFloat(Value=&FloatFromCurve(
      FloatCurve=DataInterface() {propertyValues {Curve {keys [
          {interpMode=RCIM_Cubic, tangentMode=RCTM_User, time=0, value=0}
          {arriveTangent=2, interpMode=RCIM_Cubic, leaveTangent=2, time=0.1, value=0.975}
          {interpMode=RCIM_Cubic, tangentMode=RCTM_User, time=0.5, value=1}
          {arriveTangent=-2, interpMode=RCIM_Cubic, leaveTangent=-2, time=0.9, value=1}
          {tangentMode=RCTM_User, time=1, value=0}
         ]}}}
      CurveIndex=&RemapRange(Input Value To Remap=$User.RightDot, Input Min=-1, Input Max=1, Output Min=0, Output Max=1)
     )))
   &SolveForcesAndVelocity(Clamp Velocity=false, Limit Acceleration=false)
   &ParticleState()
   &ScaleMeshSize(Scale Factor=&VectorFromFloat(Value=&FloatFromCurve(FloatCurve=DataInterface() {propertyValues {Curve {keys [
          {tangentMode=RCTM_User, time=0, value=1}
          {arriveTangent=-0.0405, interpMode=RCIM_Cubic, leaveTangent=-0.0405, tangentMode=RCTM_User, time=0.7009, value=1}
          {arriveTangent=-8.0089, interpMode=RCIM_Cubic, leaveTangent=-8.0089, tangentMode=RCTM_User, time=1, value=0}
         ]}}})))
   &DynamicMaterialParameters(
    Index 0 Param 2=1
    Index 0 Param 3=1
    Index 0 Param 4=1
    Module.Param0WriteEnabled=true
    Module.Param1WriteEnabled=true
    Module.Param2WriteEnabled=true
    Module.Param3WriteEnabled=true
   )
   &ScaleColor(
    Scale Alpha=&MakeCustomFloatFromBool(
     Bool=$User.bIsFrontEnd
     True Float=&FloatFromCurve(FloatCurve=DataInterface() {propertyValues {Curve {keys [<at 0 1>,<at 1 0>]}}})
     False Float=&FloatFromCurve(
      FloatCurve=DataInterface() {propertyValues {Curve {keys [
          {interpMode=RCIM_Cubic, tangentMode=RCTM_User, time=0, value=0}
          {interpMode=RCIM_Cubic, tangentMode=RCTM_User, time=0.5, value=1.5}
          {tangentMode=RCTM_User, time=1, value=0}
         ]}}}
      Scale Curve=&Lerp_Float(
       A=&FN_FrontEndFloatModifier(Gameplay=0, Front End=0.5)
       B=&Lerp_Float(A=0, B=0.5, Alpha=$System.Timeline001.Time)
       Alpha=$System.Timeline.Time
      )
     )
    )
    ScaleA=true
    ScaleRGB=true
   )
   &ScaleColor(Scale Alpha=&FN_FrontEndFloatModifier(Gameplay=1, Front End=0.2), ScaleA=true, ScaleRGB=true) {module=ScaleColor001}
  ]
  rendererClasses [&NiagaraMeshRendererProperties]
  renderers [{rendererClass=&NiagaraMeshRendererProperties, rendererIndex=0}]
 }
 NE_Wind_01 {
  emitterUpdateScript [
   &EmitterState(
    Inactive Response=::NewEnumerator1
    Life Cycle Mode=::NewEnumerator1
    Loop Behavior=::NewEnumerator1
    Loop Delay=2.5
    Loop Duration Mode=::NewEnumerator1
    UseLoopDelay=true
   )
   &SpawnBurst_Instantaneous(
    Spawn Probability=&MakeCustomFloatFromBool(Bool=$User.bIsFrontEnd, True Float=0, False Float=1)
    Use Spawn Probability=true
    UseLoopCountLimit=false
   )
  ]
  particleSpawnScript [
   &FN_SetSpawnGroup(Number of Spawn Groups=2)
   &InitializeParticle(
    Color=&MakeCustomLinearColorFromBool(Bool=$User.bIsFrontEndPreview, True Color=<rgba 1 1 1 0.2>, False Color=<rgba 1 1 1 1>)
    Lifetime=0.5
    Mesh Scale=<0.95 2 0.95>
    Mesh Scale Mode=::NewEnumerator4
    UsePositionOffset=false
    Write Material Random=true
   )
   &InitialMeshOrientation(Mesh Orientation Mode=::NewEnumerator2, Orientation Vector=<0 -1 0>, Rotation=<0.05 0 0>, Use Rotation Vector=true)
  ]
  particleUpdateScript [
   &ParticleState(Kill Particles When Lifetime Has Elapsed=false, Let Infinitely Lived Particles Die When Emitter Deactivates=true)
   &SkeletalMeshLocation(
    MaskSpawnGroup=false
    Mesh Sampling Type=::NewEnumerator5
    Sampled Position Offset=<-180 0 5>
    Skeletal Mesh=$System.Skeletal Mesh
    Socket Sampling Mode=::NewEnumerator5
   )
   &ScaleColor(
    Scale Alpha=&Lerp_Float(
     A=&MakeCustomFloatFromBool(Bool=$User.bIsFrontEnd, True Float=1, False Float=0)
     B=&FNTimeOfDayScalar(Night Scale=0.75, Morning Scale=0.75, Day Scale=0.75, Evening Scale=0.75)
     Alpha=$System.Timeline.NormalizedTime
    )
    Scale RGB=&VectorFromFloat(Value=&FNTimeOfDayScalar(Night Scale=0.8, Morning Scale=0.8, Day Scale=0.8, Evening Scale=0.8))
    ScaleA=true
    ScaleRGB=true
   )
  ]
  rendererClasses [&NiagaraMeshRendererProperties]
  renderers [{rendererClass=&NiagaraMeshRendererProperties, rendererIndex=0}]
 }
 NE_Wind_02 {
  emitterUpdateScript [
   &EmitterState(
    Inactive Response=::NewEnumerator1
    Life Cycle Mode=::NewEnumerator1
    Loop Behavior=::NewEnumerator1
    Loop Delay=2.5
    Loop Duration Mode=::NewEnumerator1
    UseLoopDelay=true
   )
   &SpawnBurst_Instantaneous(Use Spawn Probability=true, UseLoopCountLimit=false)
  ]
  particleSpawnScript [
   &FN_SetSpawnGroup(Number of Spawn Groups=2)
   &InitializeParticle(
    Color=&MakeCustomLinearColorFromBool(Bool=$User.bIsFrontEndPreview, True Color=<rgba 1 1 1 0.2>, False Color=<rgba 1 1 1 1>)
    Lifetime=0.5
    Mesh Scale=<0.65 0.85 0.85>
    Mesh Scale Mode=::NewEnumerator4
    UsePositionOffset=false
    Write Material Random=true
   )
   &InitialMeshOrientation(Mesh Orientation Mode=::NewEnumerator2, Orientation Vector=<0 -1 0>, Rotation=<0.15 0 0>, Use Rotation Vector=true)
  ]
  particleUpdateScript [
   &ParticleState(Kill Particles When Lifetime Has Elapsed=false, Let Infinitely Lived Particles Die When Emitter Deactivates=true)
   &SkeletalMeshLocation(
    MaskSpawnGroup=false
    Mesh Sampling Type=::NewEnumerator5
    Sampled Position Offset=<-30 0 15>
    Skeletal Mesh=$System.Skeletal Mesh
    Socket Sampling Mode=::NewEnumerator5
   )
   &ScaleColor(
    Scale Alpha=&Lerp_Float(
     A=&MakeCustomFloatFromBool(Bool=$User.bIsFrontEnd, True Float=1, False Float=0)
     B=&FNTimeOfDayScalar(Night Scale=0.75, Morning Scale=1, Evening Scale=1)
     Alpha=$System.Timeline.NormalizedTime
    )
    Scale RGB=&VectorFromFloat(Value=&FNTimeOfDayScalar(Night Scale=0.8, Morning Scale=0.8, Day Scale=0.8, Evening Scale=0.8))
    ScaleA=true
    ScaleRGB=true
   )
  ]
  rendererClasses [&NiagaraMeshRendererProperties]
  renderers [{rendererClass=&NiagaraMeshRendererProperties, rendererIndex=0}]
 }
}
summary {
 systemName=NS_Glider_ClumsyChew_Trail
 emitters [
  {emitterName=NE_Wind_01, rendererClasses [&NiagaraMeshRendererProperties]}
  {emitterName=NE_Ribbons_Front, rendererClasses [&NiagaraRibbonRendererProperties]}
  {emitterName=NE_SpeedLines, rendererClasses [&NiagaraMeshRendererProperties]}
  {emitterName=NE_Wind_02, rendererClasses [&NiagaraMeshRendererProperties]}
  {emitterName=NE_Ribbons_Back, rendererClasses [&NiagaraRibbonRendererProperties]}
 ]
 userVariables [
  User.bIsFrontEnd {type=NiagaraBool}
  User.bIsFrontEndPreview {type=NiagaraBool}
  User.bIsFullyDeployed {type=NiagaraBool}
  User.ForwardDot {type=NiagaraFloat, defaultValue=0, description=""}
  User.RibbonIDOffset {type=NiagaraInt32}
  User.RightDot {type=NiagaraFloat, defaultValue=0, description=""}
  User.VariantIndex {type=NiagaraFloat, defaultValue=0, description=""}
 ]
}
